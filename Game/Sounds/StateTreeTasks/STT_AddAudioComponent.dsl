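; StateTree task that gives the context Actor a temporary AudioComponent for the duration of a state.
; On entering the state it adds the component by class, publishes it through the Audio Component output and finishes successfully, or fails if no Actor is available.
; On exiting the state it destroys the stored component, so the component's lifetime is tied to the state.
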(event EventEnterState
  (params (Transition :type State%20Tree%20Transition%20Result%20Structure%20%28by%20ref%29))
  (bind _actor (Variables|Context|GetActor)
    (:then
      (bind _returnvalue (Game|AddComponentbyClass :self _actor :Class "/Script/Engine.AudioComponent"))
      (Variables|Output|SetAudio%20Component _returnvalue)
      (StateTree|FinishTask true))
    (:else
      (StateTree|FinishTask false))))

(event EventExitState
  (params (Transition :type State%20Tree%20Transition%20Result%20Structure%20%28by%20ref%29))
  (bind _audio_component (Variables|Output|GetAudio%20Component)
    (:then
      (Components|DestroyComponent _audio_component))
    (:else)))
