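; A StateTree task that relays an input Audio Component's OnAudioFinished event to the state tree as a delegate.
; On state entry it binds to the event and finishes immediately as succeeded, or as failed if the component is unavailable.
; When the audio finishes, it broadcasts its On Audio Finished delegate, and it unbinds on state exit.
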(event EventEnterState
  (params (Transition :type State%20Tree%20Transition%20Result%20Structure%20%28by%20ref%29))
  (bind _audio_component (Variables|Input|GetAudio%20Component)
    (:then
      (EventDispatchers|BindEventtoOnAudioFinished _audio_component (EventDispatchers|CreateEvent :!event Custom|On%20Finished))
      (StateTree|FinishTask true))
    (:else
      (StateTree|FinishTask false))))

(event EventExitState
  (params (Transition :type State%20Tree%20Transition%20Result%20Structure%20%28by%20ref%29))
  (bind _audio_component (Variables|Input|GetAudio%20Component)
    (:then
      (EventDispatchers|UnbindEventfromOnAudioFinished _audio_component (EventDispatchers|CreateEvent :!event Custom|On%20Finished)))
    (:else)))

(event Custom|On%20Finished
  (StateTree|BroadcastDelegate (Variables|Delegates|GetOn%20Audio%20Finished)))
