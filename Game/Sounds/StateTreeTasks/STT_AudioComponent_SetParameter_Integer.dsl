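; StateTree task that sets a named integer parameter on an audio component when its state is entered.
; On EnterState it binds the input Audio Component and, if present, calls SetIntegerParameter through the audio parameter controller interface using the Parameter Name and Value properties.
; It finishes immediately, reporting success when the component exists and failure when it does not.
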(event EventEnterState
  (params (Transition :type State%20Tree%20Transition%20Result%20Structure%20%28by%20ref%29))
  (bind _audio_component (Variables|Input|GetAudio%20Component)
    (:then
      (Audio|Parameter|SetIntegerParameter _audio_component (Variables|Parameters|GetParameter%20Name) (Variables|Parameters|GetValue) :!class /Script/AudioExtensions.AudioParameterControllerInterface)
      (StateTree|FinishTask true))
    (:else
      (StateTree|FinishTask false))))
