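; StateTree task that relays an audio gameplay volume proxy's listener enter and exit events into the tree.
; On enter it binds to the input proxy's OnToggledOn and OnToggledOff, then finishes immediately, successfully if the proxy exists and failed otherwise.
; Those events broadcast On Listener Enter or Exit delegates, then set the Is Listener in AGV output.
; Bindings outlive the task's completion and are removed on state exit.
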(event EventEnterState
  (params (Transition :type State%20Tree%20Transition%20Result%20Structure%20%28by%20ref%29))
  (bind _volume_proxy (Variables|Input|GetVolume%20Proxy)
    (:then
      (Events|BindEventtoOnToggledOn _volume_proxy (EventDispatchers|CreateEvent :!event Custom|On%20Listener%20Enter%20AGV))
      (Events|BindEventtoOnToggledOff _volume_proxy (EventDispatchers|CreateEvent :!event Custom|On%20Listener%20Exit%20AGV))
      (StateTree|FinishTask true))
    (:else
      (StateTree|FinishTask false))))

(event EventExitState
  (params (Transition :type State%20Tree%20Transition%20Result%20Structure%20%28by%20ref%29))
  (bind _volume_proxy (Variables|Input|GetVolume%20Proxy)
    (:then
      (Events|UnbindEventfromOnToggledOn _volume_proxy (EventDispatchers|CreateEvent :!event Custom|On%20Listener%20Enter%20AGV))
      (Events|UnbindEventfromOnToggledOff _volume_proxy (EventDispatchers|CreateEvent :!event Custom|On%20Listener%20Exit%20AGV)))
    (:else)))

(event Custom|On%20Listener%20Enter%20AGV
  (StateTree|BroadcastDelegate (Variables|Delegates|GetOn%20Listener%20Enter))
  (Variables|Output|SetIs%20Listener%20in%20AGV true))

(event Custom|On%20Listener%20Exit%20AGV
  (StateTree|BroadcastDelegate (Variables|Delegates|GetOn%20Listener%20Exit))
  (Variables|Output|SetIs%20Listener%20in%20AGV false))
