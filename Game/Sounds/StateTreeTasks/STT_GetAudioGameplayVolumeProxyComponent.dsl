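; StateTree task that looks up the AudioGameplayVolumeComponent on an input actor and exposes it as the Volume Proxy output.
; On entering the state it resolves the actor, fetches the component by class and stores it in that output.
; It finishes immediately, succeeding only if the component is valid and failing when the actor or component is missing.
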(event EventEnterState
  (params (Transition :type State%20Tree%20Transition%20Result%20Structure%20%28by%20ref%29))
  (bind _actor (Variables|Input|GetActor)
    (:then
      (bind _output_get (Variables|Output|SetVolume%20Proxy (Actor|GetComponentbyClass _actor "/Script/AudioGameplayVolume.AudioGameplayVolumeComponent")))
      (StateTree|FinishTask (Utilities|IsValid _output_get :!class /Script/Engine.KismetSystemLibrary)))
    (:else
      (StateTree|FinishTask false))))
